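# Unlit Landmass brush material that shapes landscape height for islands, working with packed RG8 height data.
# Negated world distance from a cached distance field, divided by CurveRampWidth, samples the CurveRT falloff curve with half-texel edge clamping.
# The inverted curve lerps unpacked HeightRT, routed through BrushBlendModes, toward the field's Water Height before repacking.
# BrushBlendModes receives only the existing height and zero; WaterBodyZ is unconnected.
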Material /Landmass/Landscape/BlueprintBrushes/Materials/WaterPlugin/MeshBrush_Island
graph {
 properties {bAllowNegativeEmissiveColor=true, pixelDepthOffsetMode=PDOM_Legacy, shadingModel=MSM_Unlit, usedShadingModels=Unlit}
 nodes [
  n1 = MaterialFunctionCall!(RG=ComponentMask(None=TextureSampleParameter2D() {parameterName=HeightRT, texture=/Engine/EngineResources/Black_Low.Black}.RGB) {g=true, r=true}) {materialFunction=/Landmass/Landscape/OldPrototype_BP/Mats/MF/UnpackRG8Height}
  n2 = Constant()
  n3 = MaterialFunctionCall!(Cached DF, UV Offset) {materialFunction=/Landmass/Landscape/BlueprintBrushes/MF/ReadCachedDistanceField}
  n4 = Saturate(None=Divide(A=Multiply(A=@n3.World Distance) {constB=-1}, B=ScalarParameter() {defaultValue=500, parameterName=CurveRampWidth}))
  n5 = Divide() {constA=0.5, constB=256}
  n6 = OneMinus(None=@n5)
  n7 = Saturate(None=TextureSampleParameter2D!(UVs=Clamp(None=@n4, Min=@n5, Max=@n6), Level=Constant(), Apply View MipBias) {mipValueMode=TMVM_MipLevel, parameterName=CurveRT, texture=/Landmass/Landscape/Curves/DynamicCurve_RT}.R)
  MaterialFunctionCall!(Height Int -32k to 32k=LinearInterpolate(
    A=MaterialFunctionCall!(AlphaBlend=@n1.Result, Min=@n1.Result, Max=@n1.Result, Additive=@n2) {materialFunction=/Landmass/Landscape/BlueprintBrushes/MF/BrushBlendModes}.Result
    B=@n3.Water Height
    Alpha=OneMinus(None=@n7)
   )) {materialFunction=/Landmass/Landscape/BlueprintBrushes/MF/PackRG8Height}
  ScalarParameter() {parameterName=WaterBodyZ}
 ]
}
