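# Unlit Landmass brush material that shapes landscape height around a mesh brush using a cached distance field.
# It adds masked texture displacement to a fixed or cached height, then blends with the unpacked HeightRT through an edge-width falloff, optionally remapped by a curve texture.
# Smooth min/max intersections feed blend modes and terracing.
# The result is conditionally capped using CombinedAlphaAndHeightRT, then packed into RG8 height.
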Material /Landmass/Landscape/BlueprintBrushes/Materials/WaterPlugin/MeshBrush_Width
graph {
 properties {bAllowNegativeEmissiveColor=true, pixelDepthOffsetMode=PDOM_Legacy, shadingModel=MSM_Unlit, usedShadingModels=Unlit}
 nodes [
  MaterialExpressionTextureObjectParameter_0 = TextureObjectParameter() {parameterName=BrushRT, texture=/Engine/MapTemplates/TilingNoise05}
  MaterialExpressionConstant_1 = Constant() {r=512}
  MaterialExpressionDivide_0 = Divide(B=@MaterialExpressionConstant_1)
  Custom!(Tex=@MaterialExpressionTextureObjectParameter_0, UV, Size=@MaterialExpressionConstant_1, InvSize=@MaterialExpressionDivide_0) {
   code="return Texture2DSampleBicubic(Tex, TexSampler, UV, Size, InvSize);"
   description=Common.ush Cubic Texture Sampler
   outputType=CMOT_Float1
   inputs [
    {
     inputName=Tex
     input {expression=&self:MaterialExpressionTextureObjectParameter_0}
    }
    {inputName=UV, input {expression=null}}
    {inputName=Size, input {expression=&self:MaterialExpressionConstant_1}}
    {inputName=InvSize, input {expression=&self:MaterialExpressionDivide_0}}
   ]
  }
  n1 = TextureSampleParameter2D() {parameterName=CombinedAlphaAndHeightRT, texture=/Engine/EngineResources/Black_Low.Black}
  n2 = Multiply(A=OneMinus(None=@n1.A)) {constB=1000}
  Divide(A=@n2, B=CollectionParameter() {collection=/Landmass/Landscape/BlueprintBrushes/MPC/MPC_Landscape, parameterName=LandscapeZScale})
  n3 = ScalarParameter() {defaultValue=2, parameterName=T}
  MaterialFunctionCall!(Tiling=@n3) {materialFunction=/Landmass/Landscape/OldPrototype_BP/Mats/MF/CatmullRomTexture}
  n4 = Subtract(A=Floor(None=@n1.B), B=@n2) {constB=0}
  n5 = MaterialFunctionCall!(Cached DF, UV Offset) {materialFunction=/Landmass/Landscape/BlueprintBrushes/MF/ReadCachedDistanceField}
  n6 = LinearInterpolate(A=ScalarParameter() {defaultValue=1394, parameterName=Z}, B=@n5.Terrain Height, Alpha=ScalarParameter() {parameterName=UseMeshDepth})
  n7 = TextureSampleParameter2D(UVs=Multiply(A=TextureCoordinate(), B=@n3)) {parameterName=BrushRT, texture=&TilingNoise05}
  n8 = VectorParameter() {parameterName=DisplacementChannel, defaultValue=<rgba 1 0 0 0>}
  n9 = DotProduct(A=AppendVector(A=@n7.RGB, B=@n7.RGB), B=AppendVector(A=@n8.RGB, B=@n8.RGB))
  n10 = ScalarParameter() {defaultValue=0.5, parameterName=Displacement Midpoint}
  n11 = Subtract(A=@n9, B=@n10)
  n12 = ScalarParameter() {defaultValue=256, parameterName=BrushTexHeight}
  n13 = Add(
   A=@n6
   B=Multiply(
    A=MaterialFunctionCall!(UV Distance=@n5.World Distance) {materialFunction=/Landmass/Landscape/OldPrototype_BP/Mats/MF/BrushDisplaceMask}.Mask
    B=Multiply(A=@n11, B=@n12)
   ) {desc=Displacement * Mask}
  ) {desc=Height + Displacement}
  n14 = Add(A=Multiply(A=@n5.World Distance), B=ScalarParameter() {parameterName=EdgeCenterOffset})
  n15 = MaterialFunctionCall!(A=@n14, B=Constant(), K=VectorParameter() {parameterName=SmoothingParams, defaultValue=<rgba 1 1 1 1>}.R) {materialFunction=/Landmass/Landscape/OldPrototype_BP/Mats/MF/SmoothIntersections}
  n16 = ScalarParameter() {parameterName=CapShapeInterior}
  n17 = Multiply(A=Divide(A=LinearInterpolate(A=@n14, B=@n15.Smooth Min, Alpha=@n16), B=ScalarParameter() {defaultValue=1024, parameterName=EdgeWidth})) {constB=-1}
  n18 = Saturate(None=@n17)
  n19 = Divide() {constA=0.5, constB=256}
  n20 = OneMinus(None=@n19)
  n21 = TextureSampleParameter2D!(UVs=OneMinus(None=Clamp(None=@n18, Min=@n19, Max=@n20)), Level=Constant(), Apply View MipBias) {mipValueMode=TMVM_MipLevel, parameterName=CurveRT, texture=&Black}
  n22 = Subtract(A=@n17) {constB=0.5}
  n23 = Ceil(None=Subtract(A=Abs(None=@n22)) {constB=0.5})
  n24 = LinearInterpolate(
   A=@n17
   B=LinearInterpolate(A=OneMinus(None=@n21.R), B=@n17, Alpha=Saturate(None=@n23))
   Alpha=VectorParameter() {parameterName=CurveParams}.A
  )
  n25 = MaterialFunctionCall!(RG=ComponentMask(None=TextureSampleParameter2D!(UVs, Apply View MipBias) {parameterName=HeightRT, texture=&Black}.RGB) {g=true, r=true}) {materialFunction=/Landmass/Landscape/OldPrototype_BP/Mats/MF/UnpackRG8Height}
  n26 = MaterialFunctionCall!(AlphaBlend=@n25.Result, Min=@n25.Result, Max=@n25.Result, Additive=Constant()) {materialFunction=/Landmass/Landscape/OldPrototype_BP/Mats/MF/BrushBlendModes_Manual}
  n27 = LinearInterpolate(A=@n13, B=@n26.Result, Alpha=@n24)
  n28 = LinearInterpolate(A=@n13, B=@n26.Result, Alpha=Min(A=@n24))
  n29 = Max(A=@n27, B=@n28)
  n30 = VectorParameter() {parameterName=SmoothingParams, defaultValue=<rgba 1 1 1 1>}
  n31 = Subtract(A=@n13, B=@n26.Result) {desc=Intersection of Terrain and Brush}
  n32 = Multiply(A=@n31) {constB=-1}
  n33 = ScalarParameter() {parameterName=UseMeshDepth}
  n34 = LinearInterpolate(A=@n30.G, B=Min(A=Max(A=@n32) {constB=0}, B=@n30.G), Alpha=@n33)
  n35 = Min(A=@n27, B=@n28)
  n36 = LinearInterpolate(A=@n35, B=Min(A=@n35, B=@n13), Alpha=ScalarParameter() {parameterName=UseMeshDepth})
  n37 = LinearInterpolate(A=@n30.G, B=Min(A=Max(A=@n31) {constB=0}, B=@n30.G), Alpha=@n33)
  n38 = MaterialFunctionCall(
   A=MaterialFunctionCall(A=@n26.Result, B=@n29, K=@n34) {materialFunction=&SmoothIntersections}.Smooth Min
   B=@n36
   K=@n37
  ) {materialFunction=&SmoothIntersections}
  n39 = MaterialFunctionCall(RG=ComponentMask(None=TextureSampleParameter2D() {parameterName=HeightRT, texture=&Black}.RGB) {g=true, r=true}) {materialFunction=&UnpackRG8Height}
  n40 = MaterialFunctionCall!(
   Height=MaterialFunctionCall(
    AlphaBlend=@n38.Smooth Max
    Min=MaterialFunctionCall(A=@n26.Result, B=@n29, K=@n34) {desc=Smooth Min, materialFunction=&SmoothIntersections}.Smooth Min
    Max=MaterialFunctionCall(A=@n26.Result, B=@n36, K=@n37) {desc=Smooth Max, materialFunction=&SmoothIntersections}.Smooth Max
    Additive=Add(A=@n38.Smooth Max, B=@n39.Result)
   ) {materialFunction=&BrushBlendModes_Manual}.Result
   DF Mask=@n5.World Distance
  ) {materialFunction=/Landmass/Landscape/OldPrototype_BP/Mats/MF/BrushTerracing}
  n41 = Floor(None=@n1.A)
  MaterialFunctionCall!(Height Int -32k to 32k=LinearInterpolate(A=Min(A=@n4, B=@n40.Result), B=@n40.Result, Alpha=Saturate(None=@n41))) {materialFunction=/Landmass/Landscape/BlueprintBrushes/MF/PackRG8Height}
  MaterialFunctionCall!(
   UV Distance=@n5.World Distance
   Displacement=Multiply(A=Subtract(A=@n9, B=@n10), B=ScalarParameter() {defaultValue=1, parameterName=WeightmapInfluence})
   Z Height=@n6
  ) {materialFunction=/Landmass/Landscape/OldPrototype_BP/Mats/MF/BrushAlphaMask}
  ScalarParameter() {defaultValue=1, desc=Tangent, parameterName=FalloffTangent}
 ]
}
